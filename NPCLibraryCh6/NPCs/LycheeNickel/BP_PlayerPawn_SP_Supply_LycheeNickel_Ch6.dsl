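; Player pawn Blueprint for the LycheeNickel Supply specialist NPC in Chapter 6, adding speech bubble barks on damage and on hire.
; Server damage at or above a threshold plays the Season 37 Specialist Damaged bark; the Delay after it gates nothing.
; Pawn conversion calls the parent handler, waits briefly, then plays the Hire bark on the converted pawn.
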(event EventBeginPlay
  (|CallParentFunction) ; This node is disabled and will not be called.\nDrag off pins to build functionality.
)

(event Collision|EventActorBeginOverlap
  (params (OtherActor :type Actor%20Object%20Reference))
  (|CallParentFunction OtherActor) ; This node is disabled and will not be called.\nDrag off pins to build functionality.
)

(event EventTick
  (params (DeltaSeconds :type Float%20%28single-precision%29))
  (|CallParentFunction DeltaSeconds) ; This node is disabled and will not be called.\nDrag off pins to build functionality.
)

(event EventOnDamageServer
  (params (Damage :type Float%20%28single-precision%29) (DamageTags :type Gameplay%20Tag%20Container%20Structure%20%28by%20ref%29) (Momentum :type Vector :default "0, 0, 0") (HitInfo :type Hit%20Result%20Structure%20%28by%20ref%29) (InstigatedBy :type Controller%20Object%20Reference) (DamageCauser :type Actor%20Object%20Reference) (EffectContext :type Gameplay%20Effect%20Context%20Handle%20Structure))
  ; NPC Bark OnDamage
  (if (>= Damage 12.0)
    (Class|BPFLAICommonFunctions|PlaySpeechBubble self "(TagName=\"Conversation.Entry.Season37.Specialist.Speechbubble.Damaged\")")
    (Utilities|FlowControl|Delay 15.0)))

(event Conversion|EventHandlePawnConverted
  (params (InstigatorPawn :type Fort%20Pawn%20Object%20Reference) (ConvertedPawn :type Fort%20Pawn%20Object%20Reference))
  ; NPC Bark On Hire
  (|CallParentFunction InstigatorPawn ConvertedPawn)
  (Utilities|FlowControl|Delay 4.0)
  (Class|BPFLAICommonFunctions|PlaySpeechBubble ConvertedPawn "(TagName=\"Conversation.Entry.Season37.Specialist.Speechbubble.Hire\")"))

(fn ConstructionScript
  (|CallParentFunction))
