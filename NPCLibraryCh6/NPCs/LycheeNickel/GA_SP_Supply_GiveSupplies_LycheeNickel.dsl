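; Gameplay Ability for the Lychee Nickel NPC's supply-giving behavior, inheriting its core logic from a parent ability.
; On the start-drop event it runs the parent logic, then plays a speech bubble bark on the bot pawn using the drop-supplies bark tag.
; The activate and end-ability overrides are unused stubs.
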(event Ability|EventActivateAbility
  (|CallParentFunction) ; This node is disabled and will not be called.\nDrag off pins to build functionality.
)

(event Ability|EventOnEndAbility
  (params (bWasCancelled :type Boolean))
  (|CallParentFunction bWasCancelled) ; This node is disabled and will not be called.\nDrag off pins to build functionality.
)

(event EventStartDrop
  (|CallParentFunction)
  ; Play bark for dropping an item
  (Class|BPFLAICommonFunctions|PlaySpeechBubble (Variables|InternalData|GetBotPawn) (Variables|Default|GetDropSuppliesBarkTag)))
